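# Physics asset for the Red Riding Summer backpack cosmetic, defining collision bodies and joint limits for its dangling parts.
# The longest chain runs from the zipper base through the zipper and two chain links into a four-segment tail, with swing limits widening toward the tip.
# Straps hang from the pack, while gun and knife pivot on small base bodies, the knife nearly rigid.
# Root and trap capsules have no constraints.
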PhysicsAsset /BRCosmetics/Accessories/FORT_Backpacks/Backpack_F_MED_Red_Riding_Summer/Meshes/F_MED_Red_Riding_Summer_Pack_Physics
bodies {
 dyn_chain_01 [{shapeType=Capsule, length=0.505, radius=0.505, center=<0.4646 0 0>, rotation=<rot 90 0 0>}]
 dyn_chain_02 [{shapeType=Capsule, length=0.505, radius=0.505, center=<0.7637 0 0>, rotation=<rot 90 0 0>}]
 dyn_gun_base [{shapeType=Box, extentX=0.5, extentY=0.5, extentZ=0.5}]
 dyn_gun_parent [{shapeType=Box, extentX=5.5705, extentY=9.4782, extentZ=4.2091, center=<1.5689 0 0>}]
 dyn_knife_base [{shapeType=Box, extentX=0.75, extentY=0.75, extentZ=0.75}]
 dyn_knife_parent [{shapeType=Capsule, length=17.2092, radius=2.908, center=<3.5828 0 0>, rotation=<rot 90 0 0>}]
 dyn_strap [{shapeType=Box, extentX=3.0331, extentY=5.2067, extentZ=1.223, center=<1.5678 0 0>}]
 dyn_tail_01 [{shapeType=Capsule, length=1.7687, radius=2.193, center=<3.0686 -0.0001 0.2644>, rotation=<rot -81.9636 0 0>}]
 dyn_tail_02 [{shapeType=Capsule, length=2.5, radius=3.875, center=<4.9443 0.0001 -0.7299>, rotation=<rot -81.7844 -180 180>}]
 dyn_tail_03 [{shapeType=Capsule, length=3.1172, radius=3.4331, center=<4.9419 0 -0.4256>, rotation=<rot -85.0708 180 -180>}]
 dyn_tail_04 [{shapeType=Capsule, length=3.1093, radius=2.3562, center=<3.8443 0 -0.5132>, rotation=<rot -82.3909 180 -180>}]
 dyn_top_strap [{shapeType=Box, extentX=5.4312, extentY=8.4635, extentZ=2.2335, center=<2.4295 0 0>}]
 dyn_trap_base [{shapeType=Capsule, length=7.25, radius=5.25, center=<-2.6851 7.6391 -5.477>}]
 dyn_zipper_base [{shapeType=Box, extentX=0.5, extentY=0.5, extentZ=0.5}]
 dyn_zipper_parent [{shapeType=Capsule, length=1.875, radius=0.755, center=<1.6494 0 0>, rotation=<rot 90 0 0>}]
 pack [{shapeType=Box, extentX=0.75, extentY=0.75, extentZ=0.75}]
 root [{shapeType=Capsule, length=15, radius=15, center=<-14.4305 13.2156 -9.8548>}]
}
constraints [
 {bone1Name=dyn_strap, bone2Name=pack, swing1LimitDegrees=20, swing2LimitDegrees=30, twistLimitDegrees=2}
 {bone1Name=dyn_gun_parent, bone2Name=dyn_gun_base, swing1LimitDegrees=20, swing2LimitDegrees=20, twistLimitDegrees=5}
 {bone1Name=dyn_zipper_parent, bone2Name=dyn_zipper_base, swing1LimitDegrees=20, swing2LimitDegrees=20, twistLimitDegrees=3}
 {bone1Name=dyn_chain_01, bone2Name=dyn_zipper_parent, swing1LimitDegrees=20, swing2LimitDegrees=30, twistLimitDegrees=3}
 {bone1Name=dyn_chain_02, bone2Name=dyn_chain_01, swing1LimitDegrees=20, swing2LimitDegrees=30, twistLimitDegrees=3}
 {bone1Name=dyn_tail_01, bone2Name=dyn_chain_02, swing1LimitDegrees=30, swing2LimitDegrees=30, twistLimitDegrees=2}
 {bone1Name=dyn_top_strap, bone2Name=pack, swing1LimitDegrees=5, swing2LimitDegrees=5, twistLimitDegrees=1}
 {bone1Name=dyn_tail_02, bone2Name=dyn_tail_01, swing1LimitDegrees=30, swing2LimitDegrees=30, twistLimitDegrees=2}
 {bone1Name=dyn_tail_03, bone2Name=dyn_tail_02, swing1LimitDegrees=35, swing2LimitDegrees=35, twistLimitDegrees=2}
 {bone1Name=dyn_tail_04, bone2Name=dyn_tail_03, swing1LimitDegrees=40, swing2LimitDegrees=40, twistLimitDegrees=2}
 {bone1Name=dyn_knife_parent, bone2Name=dyn_knife_base, swing1LimitDegrees=3, swing2LimitDegrees=3, twistLimitDegrees=1}
]
