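# Physics asset for the Red Riding Summer backpack's trap mesh, giving secondary motion to its strap, gun and trap pieces.
# Every body is a single box, with small boxes on pack, gun base and trap base serving as constraint anchors.
# The strap swings from the pack and the gun from its base, with narrow twist limits.
# Both trap halves hinge on the trap base, with swing1 and twist locked and a small swing2 range.
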PhysicsAsset /BRCosmetics/Accessories/FORT_Backpacks/Backpack_F_MED_Red_Riding_Summer/Meshes/F_MED_Red_Riding_Summer_Pack_Trap_Physics
bodies {
 dyn_gun_base [{shapeType=Box, extentX=0.5, extentY=0.5, extentZ=0.5}]
 dyn_gun_parent [{shapeType=Box, extentX=5.5705, extentY=9.4782, extentZ=4.2091, center=<1.5689 0 0>}]
 dyn_strap [{shapeType=Box, extentX=3.0331, extentY=5.2067, extentZ=1.223, center=<1.5678 0 0>}]
 dyn_trap_base [{shapeType=Box, extentX=1, extentY=1, extentZ=1}]
 dyn_trap_l [{shapeType=Box, extentX=8.9625, extentY=6.2655, extentZ=1.25, center=<4.9665 -0.0001 0>}]
 dyn_trap_r [{shapeType=Box, extentX=8.9625, extentY=6.2655, extentZ=1.25, center=<4.9665 -0.0001 0>}]
 pack [{shapeType=Box, extentX=0.75, extentY=0.75, extentZ=0.75}]
}
constraints [
 {bone1Name=dyn_strap, bone2Name=pack, swing1LimitDegrees=20, swing2LimitDegrees=30, twistLimitDegrees=2}
 {bone1Name=dyn_gun_parent, bone2Name=dyn_gun_base, swing1LimitDegrees=20, swing2LimitDegrees=20, twistLimitDegrees=5}
 {
  bone1Name=dyn_trap_l
  bone2Name=dyn_trap_base
  swing1LimitDegrees=20
  swing1Motion=Locked
  swing2LimitDegrees=8
  twistLimitDegrees=1
  twistMotion=Locked
 }
 {
  bone1Name=dyn_trap_r
  bone2Name=dyn_trap_base
  swing1LimitDegrees=20
  swing1Motion=Locked
  swing2LimitDegrees=8
  twistLimitDegrees=1
  twistMotion=Locked
 }
]
